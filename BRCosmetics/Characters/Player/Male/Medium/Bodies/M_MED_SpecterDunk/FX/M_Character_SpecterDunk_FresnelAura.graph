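# Translucent unlit Niagara sprite material for the SpecterDunk character's Fresnel aura effect.
# Scene world normals drive a Fresnel glow on surfaces behind the sprite, with particle dynamic parameters for exponent and flattening and a vector fallback on mobile or low quality.
# Emissive combines particle color, panning flame noise and time-pulsed cracks.
# Opacity adds noise to Fresnel, then masks it by sprite radius, depth proximity, surface verticality and particle alpha.
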Material /BRCosmetics/Characters/Player/Male/Medium/Bodies/M_MED_SpecterDunk/FX/M_Character_SpecterDunk_FresnelAura
graph {
 properties {bUsedWithNiagaraSprites=true, blendMode=BLEND_Translucent, pixelDepthOffsetMode=PDOM_Legacy, shadingModel=MSM_Unlit, usedShadingModels=Unlit}
 nodes [
  Constant()
  n1 = NamedRerouteUsage()
  n2 = TextureCoordinate()
  n3 = VectorParameter() {group=Noise, parameterName=Noise UV Control, channelNames {a=Panner V, b=Panner U, g=Tiling V, r=Tiling U}, defaultValue=<rgba 1 1 0 0.5>}
  n4 = AppendVector(A=@n3.Tiling U, B=@n3.Tiling U)
  n5 = TextureSampleParameter2D(UVs=Panner(Coordinate=Multiply(A=@n2, B=@n4), Speed=AppendVector(A=@n3.Panner U, B=@n3.Panner U)) {speedY=1}) {
   group=Noise
   parameterName=Noise
   texture=&/T_Character_SpecterDunk_FlameNoise
  }
  n6 = Multiply(A=@n1, B=Divide(A=@n5.R, B=Constant() {r=30}) {constB=6})
  n7 = TextureSampleParameter2D(UVs=MaterialFunctionCall(Texture Scale=ScalarParameter() {defaultValue=1, parameterName=UVScale}) {materialFunction=&ScaleUVsByCenter}.UVs) {parameterName=Cracks, samplerType=SAMPLERTYPE_LinearColor, texture=/Game/Effects/Fort_Effects/Textures/T_Cracks}
  n8 = Power(Base=@n7.RGB, Exp=ScalarParameter() {defaultValue=2, parameterName=Param_1})
  n9 = Multiply(A=@n1, B=VectorParameter() {defaultValue=<rgba 1 0.1298 0 1>}.RGB)
  n10 = CameraVectorWS()
  n11 = ComponentMask(None=SceneTexture() {sceneTextureId=PPI_WorldNormal}.Color) {b=true, g=true, r=true}
  n12 = Power(Base=OneMinus(None=DotProduct(A=@n10, B=@n11)), Exp=Constant() {r=1})
  n13 = Add(A=Divide(A=Time()) {constB=0.5}, B=ParticleRandom())
  n14 = Divide(A=Add(A=Sine(None=@n13)))
  n15 = Multiply(A=Multiply(A=Multiply(A=@n8, B=@n9), B=Multiply(A=@n12, B=@n14)), B=Constant() {r=50}) {constB=3}
  n16 = ParticleColor()
  n17 = Multiply(A=ScalarParameter() {defaultValue=1, parameterName=EmissiveMult}, B=Multiply(A=@n16.RGB, B=@n16.RGB))
  MaterialFunctionCall(Emissive=Add(A=@n6, B=Add(A=@n15, B=@n17)) {constB=0.01}) {materialFunction=&MF_FX_EmissiveAdjustments}
  n18 = DynamicParameter() {defaultValue=<rgba 3 0.15 1 1>, paramNames [FresnelExponent, FlattenNormal, Param3, Param4]}
  n19 = Fresnel(
   ExponentIn=@n18.FresnelExponent
   Normal=MaterialFunctionCall(Normal=@n11, Flatness=@n18.FlattenNormal) {materialFunction=&FlattenNormal}.Result
  ) {baseReflectFraction=0, exponent=1}
  n20 = VectorParameter() {parameterName=Forward, defaultValue=<rgba 0 0 1 1>}
  n21 = QualitySwitch(Default=@n19, Low=@n20.RGB, High=@n19, Medium=@n19, Epic=@n19)
  n22 = Multiply(
   A=Add(A=@n6, B=ShadingPathSwitch(Default=@n21, Mobile=@n20.RGB))
   B=SphereMask(
    A=TextureCoordinate()
    B=Constant() {r=0.5}
    Radius=ScalarParameter() {defaultValue=0.5, parameterName=SphereMaskRadius}
    Hardness=ScalarParameter() {parameterName=SphereMaskHardness}
   ) {hardnessPercent=0}
  )
  n23 = ScalarParameter() {defaultValue=30, parameterName=VisibleDepthDistance}
  n24 = DepthFade(
   Opacity=OneMinus(None=DepthFade(FadeDistance=@n23) {fadeDistanceDefault=80})
   FadeDistance=ScalarParameter() {defaultValue=5, parameterName=DepthFade}
  ) {fadeDistanceDefault=15}
  n25 = DotProduct(
   A=Constant3Vector() {constant=<rgba 0 0 1 0>}
   B=ComponentMask(None=SceneTexture() {sceneTextureId=PPI_WorldNormal}.Color) {b=true, g=true, r=true}
  )
  n26 = Power(Base=Abs(None=@n25), Exp=ScalarParameter() {defaultValue=1, parameterName=FloorFade})
  Multiply(A=@n16.A, B=Multiply(A=Multiply(A=@n22, B=@n24), B=Saturate(None=@n26)))
  Multiply(A=BlackBody!(Temp=Constant() {r=9600}), B=@n1)
  NamedRerouteDeclaration(None=@n16.RGB) {name=Color, nodeColor=<rgba 0.6172 1 0 1>}
  OneMinus(None=@n7.RGB)
  OneMinus(None=@n25)
 ]
}
