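# Niagara idle effect for the SpecterDunk character body cosmetic, attaching flames and glows to skeletal mesh bones or sockets.
# Glows and FlamesAndDecal pin burst sprites per socket, with extra flames spawning between socket pairs; FlameLicks adds rising wisps, scaled by User.Variant.
# User.Variant also picks Heat or Surge color curves, though FlameLicks always uses Surge.
# Head-hidden, hand-hidden and DBNO flags fade or kill particles on selected sockets.
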NiagaraSystem /BRCosmetics/Characters/Player/Male/Medium/Bodies/M_MED_SpecterDunk/FX/NS_Character_SpecterDunk_Idle
dependencies {
 usedDataInterfaces [
  &NiagaraDataInterfaceCurve
  &NiagaraDataInterfaceArrayFloat
  &NiagaraDataInterfaceVector2DCurve
  &NiagaraDataInterfaceSpriteRendererInfo
  &NiagaraDataInterfaceSkeletalMesh
  &NiagaraDataInterfaceColorCurve
 ]
 usedDynamicInputs [
  &MakeCustomVector2DFromBool
  &SimulationPosition
  &MaskFloatBySpawnGroup
  &SelectFloatFromArray
 ]
 usedModules [
  &self:SystemSpawnScript.SystemScriptSource.SystemScriptGraph.NiagaraNodeAssignment_1.SetVariables_#1
  &self:SystemSpawnScript.SystemScriptSource.SystemScriptGraph.NiagaraNodeAssignment_2.SetVariables_#2
  &SystemState
  &self:SystemSpawnScript.SystemScriptSource.SystemScriptGraph.NiagaraNodeAssignment_4.SetVariables_#3
 ]
}
emitters {
 FlameLicks {
  emitterUpdateScript [
   &EmitterState(
    Inactive Response=::NewEnumerator1
    Life Cycle Mode=::NewEnumerator1
    Loop Behavior=::NewEnumerator1
    Loop Duration Mode=::NewEnumerator1
    UseLoopDelay=false
   )
   &SpawnRate(Spawn Probability=&Multiply_Float(A=$User.Variant, B=0.5), SpawnRate=15, Use Spawn Probability=true)
  ]
  particleSpawnScript [
   /SetVariables(Particles.VisibilityTag=&SpawnGroup()) {isSetParameters=true, module=SetVariables_#4}
   &InitializeParticle(
    Color=<rgba 3 3 3 0.65>
    Lifetime=&RandomRangeFloat(Minimum=0.6, Maximum=0.8)
    Mesh Scale Mode=::NewEnumerator4
    Sprite Rotation Mode=::NewEnumerator2
    Sprite Size Mode=::NewEnumerator3
    Sprite UV Mode=::NewEnumerator3
    Uniform Sprite Size=&RandomRangeFloat(Minimum=8, Maximum=15)
    UsePositionOffset=false
    Write Material Random=true
   )
   &CameraOffset()
   /SetVariables(Particles.CustomRandom:NiagaraInt32=&MergeHelper_IntSwitch(0=0, 1=1, 2=2, 3=4, 4=6, 5=?, Index=&RandomRangeInt(Minimum=0, Maximum=4), Num=::NewEnumerator3)) {isSetParameters=true, module=SetVariables_#5}
   /SetVariables() {isSetParameters=true}
  ]
  particleUpdateScript [
   &ParticleState()
   &SpriteRotationRate(Rotation Rate=&Multiply_Float(
     A=&FloatFromCurve(FloatCurve=DataInterface() {propertyValues {Curve {keys [<at 0 1 RCIM_Cubic>,<at 1 0 RCIM_Cubic>]}}})
     B=&RandomRangeFloat(Minimum=-120, Maximum=120)
    ))
   &SkeletalMeshLocation(
    Bone / Socket Index=$Particles.CustomRandom
    Bone Sampling Mode=::NewEnumerator4
    MaskSpawnGroup=false
    Position Sampling=::NewEnumerator2
    Sampled Position Offset=<0 0 -7>
    Sampled Position Offset Coordinate Space:ENiagaraSkelMeshTransforms=::NewEnumerator1
    Skeletal Mesh=$System.Skeletal Mesh
   ) {module=SkeletalMeshLocation001}
   &SkeletalMeshLocation(
    Bone / Socket Index=&Add_Integer(A=$Particles.CustomRandom, B=1)
    Bone Sampling Mode=::NewEnumerator4
    MaskSpawnGroup=false
    Position Sampling=::NewEnumerator2
    Sampled Position Offset=<0 0 -7>
    Sampled Position Offset Coordinate Space:ENiagaraSkelMeshTransforms=::NewEnumerator1
    Skeletal Mesh=$System.Skeletal Mesh
   ) {module=SkeletalMeshLocation002}
   &SolveForcesAndVelocity(Clamp Velocity=false, Limit Acceleration=false)
   &ScaleColor(
    Scale Alpha=&FloatFromCurve(FloatCurve=DataInterface() {propertyValues {Curve {keys [
         <at 0 0 RCIM_Cubic>
         {arriveTangent=2, leaveTangent=2, time=0.1, value=1}
         {arriveTangent=-1.1111, interpMode=RCIM_Cubic, leaveTangent=-1.1111, time=0.5, value=1}
         <at 1 0 RCIM_Cubic>
        ]}}})
    Scale RGB=&MakeVectorFromLinearColorRGB(Color=&MergeHelper_ColorSwitch(
      0=&ColorFromCurve(
       ColorCurve=$System.HeatColor
       CurveIndex=&Frac_Float(Float=&Add_Float(A=&RandomRangeFloat(Minimum=0, Maximum=1), B=$Emitter.LoopedAge))
      )
      1=&ColorFromCurve(
       ColorCurve=$System.SurgeColor
       CurveIndex=&Frac_Float(Float=&Add_Float(A=&RandomRangeFloat(Minimum=0, Maximum=1), B=$Emitter.LoopedAge))
      )
      2=?
      3=?
      Index=1
      Num=::NewEnumerator0
     ))
    ScaleA=true
    ScaleRGB=true
   )
   &ScaleSpriteSize(
    Scale Sprite Size Mode=::NewEnumerator0
    Uniform Scale Factor=&FloatFromCurve(FloatCurve=DataInterface() {propertyValues {Curve {
        defaultValue=1
        keys [
         <at 0 0.5 RCIM_Cubic>
         {arriveTangent=-0.3, interpMode=RCIM_Cubic, leaveTangent=-0.3, time=0.2, value=1}
         {arriveTangent=-2.6646, interpMode=RCIM_Cubic, leaveTangent=-2.6646, tangentMode=RCTM_User, time=1, value=0.2}
        ]
       }}})
   )
   &SubUVAnimation(
    End Frame Range Override=15
    "Sprite Renderer "=DataInterface() {propertyValues {SpriteRenderer=&self:FlameLicks_1.NiagaraSpriteRendererProperties_2}}
    Start Frame Range Override=7
    UseEndFrame=true
    UseStartFrame=true
   )
   /SetVariables(Particles.FlameSize:Vector2f=&Multiply_Vector2D_ByFloat(Float=0.7, Vector2D=$Particles.SpriteSize)) {isSetParameters=true}
   /SetVariables(Particles.GlowColor=&MultiplyLinearColor(
     Linear Color=$Particles.Color
     Scale Factor=&MakeFloat4FromVectorAndFloat(
      "Vector (RGB)"=<1 1 1>
      "Float (Alpha)"=&FloatFromCurve(
       FloatCurve=DataInterface() {propertyValues {Curve {keys [{interpMode=RCIM_Cubic, tangentMode=RCTM_User, time=0, value=1}, {interpMode=RCIM_Cubic, tangentMode=RCTM_User, time=0.7, value=0}]}}}
       Scale Curve=0.1
      )
     )
    )) {isSetParameters=true}
   /SetVariables(Particles.Position=&AddVectorToPosition(
     Position=&AddVectorToPosition(
      Position=&Lerp_Vector(
       A=Linked() {linkedVariable {name=Particles.SkeletalMeshLocation001.SampledPosition, type=NiagaraPosition}}
       B=Linked() {linkedVariable {name=Particles.SkeletalMeshLocation002.SampledPosition, type=NiagaraPosition}}
       Alpha=&RandomRangeFloat(Minimum=0, Maximum=1)
      )
      Vector=<0 0 16>
     )
     Vector=&Lerp_Vector(
      A=<0 0 0>
      B=&RandomRangeVector(Minimum=<-0.3 -0.3 10>, Maximum=<0.3 0.3 14>)
      Alpha=&FloatFromCurve(
       FloatCurve=DataInterface() {propertyValues {Curve {
          defaultValue=1
          keys [{arriveTangent=1.8953, interpMode=RCIM_Cubic, leaveTangent=1.8953, tangentMode=RCTM_User, time=0, value=0}, <at 1 1 RCIM_Cubic>]
         }}}
       CurveIndex=&Multiply_Float(A=$Particles.Age, B=&RandomRangeFloat(Minimum=1, Maximum=1.5))
      )
     )
    )) {isSetParameters=true, module=SetVariables_#6}
   &DynamicMaterialParameters(
    Index 0 Param 1=&RandomRangeFloat(Minimum=0, Maximum=1)
    Index 0 Param 2=&FloatFromCurve(
     FloatCurve=DataInterface() {propertyValues {Curve {
        defaultValue=1
        keys [<at 0.5 0 RCIM_Cubic>, {arriveTangent=2.8965, interpMode=RCIM_Cubic, leaveTangent=2.8965, tangentMode=RCTM_User, time=1, value=1}]
       }}}
     Scale Curve=0.7
    )
    Module.Index0Float0Write=true
    Module.Index0Float1Write=true
    Module.Index0Float2Write=true
    Module.Index0Float3Write=true
   )
   &KillParticles(Kill Particles=&Boolean_OrOperation(
     Bool1=&Boolean_OrOperation(
      Bool1=&Boolean_AndOperation(
       Bool1=$User.bIsHeadHidden
       Bool2=&SetBoolByIntComparison(A=$Particles.CustomRandom, B=1, Comparison Type=::NewEnumerator0)
      )
      Bool2=&Boolean_AndOperation(
       Bool1=$User.bIsHandHidden
       Bool2=&SetBoolByIntComparison(A=$Particles.CustomRandom, B=1, Comparison Type=::NewEnumerator7)
      )
     )
     Bool2=&Boolean_AndOperation(
      Bool1=$User.IsInDBNO
      Bool2=&SetBoolByIntComparison(
       A=&MergeHelper_IntSwitch(0=0, 1=0, 2=1, 3=0, 4=1, 5=0, 6=1, Index=$Particles.CustomRandom, Num=::NewEnumerator5)
       B=1
       Comparison Type=::NewEnumerator4
      )
     )
    )) {module=KillParticles001}
  ]
  rendererClasses [&NiagaraSpriteRendererProperties]
  renderers [
   {rendererClass=&NiagaraSpriteRendererProperties}
   {rendererClass=&NiagaraSpriteRendererProperties, rendererIndex=1}
  ]
 }
 FlamesAndDecal {
  emitterUpdateScript [
   &EmitterState(
    Inactive Response=::NewEnumerator1
    Life Cycle Mode=::NewEnumerator1
    Loop Behavior=::NewEnumerator1
    Loop Duration Mode=::NewEnumerator1
    UseLoopDelay=false
   )
   &SpawnRate(Spawn Probability=0.5, SpawnGroup=1, SpawnRate=20, Use Spawn Probability=true)
   &SpawnBurst_Instantaneous(Spawn Count=8, Use Spawn Probability=false, UseLoopCountLimit=false)
  ]
  particleSpawnScript [
   /SetVariables(Particles.VisibilityTag=&SpawnGroup()) {isSetParameters=true, module=SetVariables_#4}
   &PartitionParticles(Number Of Partitions=8)
   &InitializeParticle(
    Color=<rgba 3 3 3 1>
    Mesh Scale=&Multiply_VectorByFloat(Vector=<0.3 1 0.3>, Float=&RandomRangeFloat(Minimum=0.7, Maximum=0.8))
    Mesh Scale Mode=::NewEnumerator4
    Sprite Rotation Mode=::NewEnumerator1
    Sprite Size=/Niagara/DynamicInputs/Helpers/MaskVector2d(
     Mask Source:NiagaraInt32=1
     Mask Parameter:NiagaraInt32=$Particles.PartitionParticles.Partition
     Mask:NiagaraInt32=6
     Masked Vector2D:Vector2f=&RandomRangeVector2D(Minimum=<35 40>, Maximum=<35 50>)
     Passthrough Vector2D:Vector2f=&MergeHelper_Vector2Switch(
      0=&RandomRangeVector2D(Minimum=<20 25>, Maximum=<35 30>)
      1=&RandomRangeVector2D(Minimum=<15 20>, Maximum=<20 35>)
      2=?
      3=?
      4=?
      Index=&SpawnGroup()
      Num=::NewEnumerator0
     )
    )
    Sprite Size Mode=::NewEnumerator4
    UsePositionOffset=false
    Write Material Random=true
   )
   /SetVariables(
    Particles.DecalOffset=10
    Particles.DecalSize:Vector2f=&MergeHelper_Vector2Switch(
     0=<30 30>
     1=<30 30>
     2=<30 30>
     3=<30 30>
     4=<30 30>
     5=<30 30>
     6=<0 0>
     7=<0 0>
     Index=$Particles.PartitionParticles.Partition
     Num=::NewEnumerator6
    )
    Particles.DynamicParameters:Vector4f=&MakeVector4(X=3, Y=0.1, Z=0, W=0)
    Particles.Random=&RandomRangeFloat(Minimum=0, Maximum=1)
   ) {isSetParameters=true}
   &InitialMeshOrientation(Mesh Orientation Mode=::NewEnumerator2, Rotation=<0 0 0.25>, Use Rotation Vector=true)
   &InitialMeshOrientation(Mesh Orientation Mode=::NewEnumerator2, Rotation=<0.25 0 0>, Use Rotation Vector=true) {module=InitialMeshOrientation001}
   &CameraOffset(Camera Offset Amount=-2)
   /SetVariables(Particles.CustomRandom:NiagaraInt32=&MergeHelper_IntSwitch(0=0, 1=1, 2=2, 3=4, 4=6, 5=?, Index=&RandomRangeInt(Minimum=0, Maximum=4), Num=::NewEnumerator3)) {isSetParameters=true, module=SetVariables_#5}
   /SetVariables() {isSetParameters=true}
  ]
  particleUpdateScript [
   &ParticleState(Kill Particles When Lifetime Has Elapsed=false)
   &SkeletalMeshLocation(
    Bone / Socket Index=$Particles.PartitionParticles.Partition
    Bone Sampling Mode=::NewEnumerator4
    MaskSpawnGroup=false
    Position Sampling=::NewEnumerator2
    Sampled Position Offset=<0 0 -7>
    Sampled Position Offset Coordinate Space:ENiagaraSkelMeshTransforms=::NewEnumerator0
    Skeletal Mesh=$System.Skeletal Mesh
   )
   &SkeletalMeshLocation(
    Bone / Socket Index=$Particles.CustomRandom
    Bone Sampling Mode=::NewEnumerator4
    MaskSpawnGroup=false
    Position Sampling=::NewEnumerator2
    Sampled Position Offset=<0 0 -7>
    Sampled Position Offset Coordinate Space:ENiagaraSkelMeshTransforms=::NewEnumerator1
    Skeletal Mesh=$System.Skeletal Mesh
   ) {module=SkeletalMeshLocation001}
   &SkeletalMeshLocation(
    Bone / Socket Index=&Add_Integer(A=$Particles.CustomRandom, B=1)
    Bone Sampling Mode=::NewEnumerator4
    MaskSpawnGroup=false
    Position Sampling=::NewEnumerator2
    Sampled Position Offset=<0 0 -7>
    Sampled Position Offset Coordinate Space:ENiagaraSkelMeshTransforms=::NewEnumerator1
    Skeletal Mesh=$System.Skeletal Mesh
   ) {module=SkeletalMeshLocation002}
   /SetVariables(Particles.ConditionalOpacity:NiagaraFloat=&ClampFloat(Float=&Add_Float(
      A=$Particles.ConditionalOpacity
      B=&MakeCustomFloatFromBool(
       Bool=&Boolean_OrOperation(
        Bool1=&Boolean_OrOperation(
         Bool1=&Boolean_AndOperation(
          Bool1=$User.bIsHeadHidden
          Bool2=&Boolean_OrOperation(
           Bool1=&SetBoolByIntComparison(
            A=&MergeHelper_IntSwitch(0=0, 1=0, 2=0, 3=1, 4=0, 5=1, 6=1, 7=1, Index=$Particles.PartitionParticles.Partition, Num=::NewEnumerator6)
            B=1
            Comparison Type=::NewEnumerator4
           )
           Bool2=&SetBoolByIntComparison(A=&Multiply_Int(A=$Particles.CustomRandom, B=$Particles.VisibilityTag), B=2, Comparison Type=::NewEnumerator0)
          )
         )
         Bool2=&Boolean_AndOperation(
          Bool1=$User.bIsHandHidden
          Bool2=&Boolean_OrOperation(
           Bool1=&SetBoolByIntComparison(
            A=&MergeHelper_IntSwitch(0=1, 1=1, 2=0, 3=0, 4=0, 5=0, 6=0, 7=0, Index=$Particles.PartitionParticles.Partition, Num=::NewEnumerator6)
            B=1
            Comparison Type=::NewEnumerator4
           )
           Bool2=&SetBoolByIntComparison(
            A=&Multiply_Int(
             A=&MergeHelper_IntSwitch(0=1, 1=1, 2=0, 3=0, 4=0, 5=?, Index=$Particles.CustomRandom, Num=::NewEnumerator3)
             B=$Particles.VisibilityTag
            )
            B=1
            Comparison Type=::NewEnumerator4
           )
          )
         )
        )
        Bool2=&Boolean_AndOperation(
         Bool1=$User.IsInDBNO
         Bool2=&SetBoolByIntComparison(
          A=&MergeHelper_IntSwitch(0=0, 1=0, 2=0, 3=1, 4=0, 5=1, 6=1, 7=0, Index=$Particles.Initial.PartitionParticles.Partition, Num=::NewEnumerator6)
          B=1
          Comparison Type=::NewEnumerator4
         )
        )
       )
       True Float=-0.2
       False Float=0.2
      )
     ))) {isSetParameters=true}
   &SolveForcesAndVelocity(Clamp Velocity=false, Limit Acceleration=false)
   &DynamicMaterialParameters(
    Index 0 Param 1=&RandomRangeFloat(Minimum=0, Maximum=0.2)
    Index 0 Param 2=&RandomRangeFloat(Minimum=0, Maximum=100)
    Module.Index0Float0Write=true
    Module.Index0Float1Write=true
    Module.Index0Float2Write=true
    Module.Index0Float3Write=true
   )
   &KillParticles(Kill Particles=&SetBoolByFloatComparison(
     A=&MergeHelper_FloatSwitch(0=0, 1=$Particles.NormalizedAge, 2=?, 3=?, Index=&SpawnGroup(), Num=::NewEnumerator0)
     B=0.99
     Comparison Type=::NewEnumerator2
    ))
   &ScaleSpriteSize(
    Scale Factor=&Multiply_Vector2D_ByFloat(
     Float=$Particles.ConditionalOpacity
     Vector2D=&MergeHelper_Vector2Switch(
      0=<1 1>
      1=&Vector2DFromCurve(
       Vector2Curve=DataInterface() {propertyValues {
         XCurve {keys [<at 0 1 RCIM_Cubic>,<at 1 0.5 RCIM_Cubic>]}
         YCurve {keys [{arriveTangent=2.5026, interpMode=RCIM_Cubic, time=0, value=0.2}, <at 1 2 RCIM_Cubic>]}
        }}
       ScaleCurve=&RandomRangeVector2D(Minimum=<1 1>, Maximum=<1 1.2>)
      )
      2=?
      3=?
      4=?
      Index=&SpawnGroup()
      Num=::NewEnumerator0
     )
    )
    Scale Sprite Size Mode=::NewEnumerator2
   )
   &ScaleColor(
    Scale Alpha=&Multiply_Float(
     A=&Multiply_Float(A=$System.SystemFade, B=$Particles.ConditionalOpacity)
     B=&MergeHelper_FloatSwitch(
      0=1
      1=&FloatFromCurve(FloatCurve=DataInterface() {propertyValues {Curve {keys [<at 0 0 RCIM_Cubic>,<at 0.1 1 RCIM_Cubic>,<at 1 0 RCIM_Cubic>]}}})
      2=?
      3=?
      Index=&SpawnGroup()
      Num=::NewEnumerator0
     )
    )
    Scale RGB=&MakeVectorFromLinearColorRGB(Color=&MergeHelper_ColorSwitch(
      0=&ColorFromCurve(
       ColorCurve=$System.HeatColor
       CurveIndex=&Frac_Float(Float=&Add_Float(A=&RandomRangeFloat(Minimum=0, Maximum=1), B=$Emitter.LoopedAge))
      )
      1=&ColorFromCurve(
       ColorCurve=$System.SurgeColor
       CurveIndex=&Frac_Float(Float=&Add_Float(A=&RandomRangeFloat(Minimum=0, Maximum=1), B=$Emitter.LoopedAge))
      )
      2=?
      3=?
      Index=&MakeIntFromFloat(Float=$User.Variant)
      Num=::NewEnumerator0
     ))
    ScaleA=true
    ScaleRGB=true
   )
   /SetVariables(Particles.DecalColor:LinearColor=&MakeLinearColorFromVectorAndFloat(
     "Vector (RGB)"=&MakeVectorFromLinearColorRGB(Color=&MergeHelper_ColorSwitch(
       0=&ColorFromCurve(
        ColorCurve=$System.HeatColor
        CurveIndex=&Frac_Float(Float=&Add_Float(A=&RandomRangeFloat(Minimum=0, Maximum=1), B=$Emitter.LoopedAge))
       )
       1=&ColorFromCurve(
        ColorCurve=$System.SurgeColor
        CurveIndex=&Frac_Float(Float=&Add_Float(A=&RandomRangeFloat(Minimum=0, Maximum=1), B=$Emitter.LoopedAge))
       )
       2=?
       3=?
       Index=&MakeIntFromFloat(Float=$User.Variant)
       Num=::NewEnumerator0
      ))
     "Float (Alpha)"=&Multiply_Float(
      A=&Multiply_Float(A=$System.SystemFade, B=$Particles.ConditionalOpacity)
      B=&Waveform(
       Global Input Position=&RandomRangeFloat(Minimum=0, Maximum=10)
       "Waveform [1]"=::NewEnumerator9
       "[1] Frequency"=&RandomRangeFloat(Minimum=3, Maximum=6)
       "[1] Amplitude Min/Max"=<1 2>
      )
     )
    )) {isSetParameters=true}
   /SetVariables(Particles.DecalSize:Vector2f=&Multiply_Vector2D_ByFloat(Float=$Particles.ConditionalOpacity, Vector2D=$Particles.Initial.DecalSize)) {isSetParameters=true}
   /SetVariables(Particles.Position=&MergeHelper_VectorSwitch(
     0=&AddVectorToPosition(
      Position=$Particles.SkeletalMeshLocation.SampledPosition
      Vector=&MaskVector(
       Mask Source=1
       Mask Parameter=$Particles.PartitionParticles.Partition
       Mask=7
       Masked Vector=&AddVectorToPosition(
        Position=<0 0 12>
        Vector=&MakeCustomVectorFromBool(
         Bool=$User.IsInDBNO
         True Vector=&RotateVectorByQuaternion(Vector To Rotate=<5 5 0>, Quaternion=$Particles.SkeletalMeshLocation.SampledRotation)
         False Vector=<0 0 0>
        )
       )
       Passthrough Vector=<0 0 7>
      )
     )
     1=&AddVectorToPosition(
      Position=&Lerp_Vector(
       A=Linked() {linkedVariable {name=Particles.SkeletalMeshLocation001.SampledPosition, type=NiagaraPosition}}
       B=Linked() {linkedVariable {name=Particles.SkeletalMeshLocation002.SampledPosition, type=NiagaraPosition}}
       Alpha=&RandomRangeFloat(Minimum=0, Maximum=1)
      )
      Vector=<0 0 8>
     )
     2=?
     3=?
     Index=&SpawnGroup()
     Num=::NewEnumerator0
    )) {isSetParameters=true, module=SetVariables_#6}
  ]
  rendererClasses [&NiagaraSpriteRendererProperties]
  renderers [
   {rendererClass=&NiagaraSpriteRendererProperties}
   {rendererClass=&NiagaraSpriteRendererProperties, rendererIndex=1}
   {rendererClass=&NiagaraSpriteRendererProperties, rendererIndex=2}
  ]
 }
 Glows {
  emitterUpdateScript [
   &EmitterState()
   &SpawnBurst_Instantaneous(Spawn Count=8, Spawn Group=1, Use Spawn Probability=false, UseLoopCountLimit=true) {module=SpawnBurst_Instantaneous001}
  ]
  particleSpawnScript [
   &PartitionParticles(Number Of Partitions=8)
   &InitializeParticle(
    Color Mode=::NewEnumerator0
    Lifetime=2
    Position Offset=<0 -30 110>
    Sprite Rotation Mode=::NewEnumerator2
    Sprite Size Mode=::NewEnumerator3
    Uniform Sprite Size=&Multiply_Float(
     A=50
     B=&MergeHelper_FloatSwitch(0=0.6, 1=1, 2=1, 3=1, 4=1, 5=1, 6=1, 7=1, Index=$Particles.PartitionParticles.Partition, Num=::NewEnumerator6)
    )
    UsePositionOffset=false
    Write Material Random=true
   ) {module=InitializeParticle001}
   /SetVariables() {isSetParameters=true}
  ]
  particleUpdateScript [
   &ParticleState(Kill Particles When Lifetime Has Elapsed=false, Let Infinitely Lived Particles Die When Emitter Deactivates=true)
   &DynamicMaterialParameters(
    Index 0 Param 1=1
    Index 0 Param 2=&RandomRangeFloat(Minimum=0, Maximum=0.2)
    Module.Index0Param0Write=true
    Module.Index0Param1Write=true
    Module.Index0Param2Write=false
    Module.Index0Param3Write=false
   )
   &SkeletalMeshLocation(
    Bone / Socket Index=$Particles.Initial.PartitionParticles.Partition
    Bone Sampling Mode=::NewEnumerator4
    MaskSpawnGroup=false
    Orientation Sampling=::NewEnumerator1
    Sampled Position Offset=&MaskVector(Mask Source=1, Mask Parameter=$Particles.PartitionParticles.Partition, Mask=7, Masked Vector=<0 0 7>, Passthrough Vector=<0 0 0>)
    Sampled Position Offset Coordinate Space:ENiagaraSkelMeshTransforms=::NewEnumerator1
    Skeletal Mesh=$System.Skeletal Mesh
   )
   &CameraOffset(Camera Offset Amount=-5)
   /SetVariables(Particles.ConditionalOpacity:NiagaraFloat=&ClampFloat(Float=&Add_Float(
      A=$Particles.ConditionalOpacity
      B=&MakeCustomFloatFromBool(
       Bool=&Boolean_OrOperation(
        Bool1=&Boolean_AndOperation(
         Bool1=$User.bIsHeadHidden
         Bool2=&SetBoolByIntComparison(
          A=&MergeHelper_IntSwitch(0=0, 1=0, 2=0, 3=1, 4=0, 5=1, 6=1, 7=1, Index=$Particles.PartitionParticles.Partition, Num=::NewEnumerator6)
          B=1
          Comparison Type=::NewEnumerator4
         )
        )
        Bool2=&Boolean_AndOperation(
         Bool1=$User.bIsHandHidden
         Bool2=&SetBoolByIntComparison(
          A=&MergeHelper_IntSwitch(0=1, 1=1, 2=0, 3=0, 4=0, 5=0, 6=0, 7=0, Index=$Particles.PartitionParticles.Partition, Num=::NewEnumerator6)
          B=1
          Comparison Type=::NewEnumerator4
         )
        )
       )
       True Float=-0.2
       False Float=0.2
      )
     ))) {isSetParameters=true}
   /SetVariables(Particles.BodyColor:LinearColor=&MultiplyLinearColor(
     Linear Color=<rgba 1 1 1 0.5>
     Scale Factor=&MakeVector4FromVectorAndFloat(
      "Vector (RGB)"=&MakeVectorFromLinearColorRGB(Color=&MergeHelper_ColorSwitch(
        0=&ColorFromCurve(
         ColorCurve=$System.HeatColor
         CurveIndex=&Frac_Float(Float=&Add_Float(A=&RandomRangeFloat(Minimum=0, Maximum=1), B=$Emitter.LoopedAge))
        )
        1=&ColorFromCurve(
         ColorCurve=$System.SurgeColor
         CurveIndex=&Frac_Float(Float=&Add_Float(A=&RandomRangeFloat(Minimum=0, Maximum=1), B=$Emitter.LoopedAge))
        )
        2=?
        3=?
        Index=&MakeIntFromFloat(Float=$User.Variant)
        Num=::NewEnumerator0
       ))
      "Float (Alpha)"=&Multiply_Float(A=$System.SystemFade, B=$Particles.ConditionalOpacity)
     )
    )) {isSetParameters=true}
   &ScaleSpriteSize(Scale Sprite Size Mode=::NewEnumerator0, Uniform Scale Factor=$Particles.ConditionalOpacity)
  ]
  rendererClasses [&NiagaraSpriteRendererProperties]
  renderers [{rendererClass=&NiagaraSpriteRendererProperties}]
 }
}
summary {
 systemName=NS_Character_SpecterDunk_Idle
 emitters [
  {emitterName=Glows, rendererClasses [&NiagaraSpriteRendererProperties]}
  {emitterName=FlamesAndDecal, rendererClasses [&NiagaraSpriteRendererProperties]}
  {emitterName=FlameLicks, rendererClasses [&NiagaraSpriteRendererProperties]}
 ]
 userVariables [
  User.bIsHandHidden {type=NiagaraBool}
  User.bIsHeadHidden {type=NiagaraBool}
  User.IsFrontend {type=NiagaraBool}
  User.IsInDBNO {type=NiagaraBool}
  User.Variant {type=NiagaraFloat, defaultValue=0, description=""}
 ]
}
