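# Parent material for the SpecterDunk player character body, inserting custom base color work between the shared 2023 character PreFX and PostFX functions.
# It composites back and chest textures over base color, positioned by offset and scale parameters, mostly grouped as Ranked.
# A masked darkening term, shaped by a sphere mask on pre-skinned local position, then partially multiplies the color.
# Packed mask textures supply arm, metal, skull and dark-skin reroutes.
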Material /BRCosmetics/Characters/Player/Male/Medium/Bodies/M_MED_SpecterDunk/Materials/M_SpecterDunk_Parent
graph {
 properties {
  bAutomaticallySetUsageInEditor=false
  bUseMaterialAttributes=true
  bUsedWithMeshDeformer=true
  bUsedWithMorphTargets=true
  bUsedWithSkeletalMesh=true
  pixelDepthOffsetMode=PDOM_Legacy
  shadingModel=MSM_FromMaterialExpression
  usedShadingModels=Default Lit | Subsurface | Clear Coat | Subsurface Profile | Cloth
 }
 nodes [
  n1 = MaterialFunctionCall() {materialFunction=&MF_CharacterParent_2023_PreFX}
  MaterialFunctionCall(
   Material=SetMaterialAttributes!(MaterialAttributes=NamedRerouteUsage(), Base Color=NamedRerouteUsage()) {attributeSetTypes [#1]}
   M=@n1.M
   Pre Skinned Local Position=@n1.M
   Distance Blend=@n1.M
   Fresnel=@n1.M
  ) {materialFunction=&MF_CharacterParent_2023_PostFX}
  NamedRerouteDeclaration(None=ComponentMask(None=@n1.M) {g=true}) {name=In_AO, nodeColor=<rgba 0 0.1797 1 1>}
  n2 = GetMaterialAttributes(None=@n1.Result) {attributeGetTypes [#1]}
  NamedRerouteDeclaration(None=@n2.Base Color) {name=IN_BaseColor, nodeColor=<rgba 0 1 0.2969 1>}
  n3 = TextureSample() {
   samplerType=SAMPLERTYPE_Masks
   texture=/BRCosmetics/Characters/Player/Male/Medium/Bodies/M_MED_SpecterDunk/Textures/T_SpecterDunk_Body_TAMask
  }
  NamedRerouteDeclaration(None=@n3.B) {name=ArmMask, nodeColor=<rgba 0.5234 0 1 1>}
  NamedRerouteDeclaration(None=@n2.MaterialAttributes) {name=In_MA, nodeColor=<rgba 1 0 0.0234 1>}
  n4 = TextureCoordinate()
  n5 = AppendVector(
   A=ScalarParameter() {defaultValue=0.016, group=Ranked, parameterName=Back_UV_OffsetX}
   B=ScalarParameter() {defaultValue=0.34, group=Ranked, parameterName=Back_UV_OffsetY}
  )
  n6 = ScalarParameter() {defaultValue=0.21, group=Ranked, parameterName=Back_UV_ScaleX}
  n7 = ScalarParameter() {defaultValue=0.22, parameterName=Back_UV_ScaleY}
  n8 = TextureSample(UVs=MaterialFunctionCall(UVs=Add(A=@n4, B=@n5), Texture Scale=AppendVector(A=@n6, B=@n7)) {materialFunction=&ScaleUVsByCenter}.UVs) {
   samplerSource=SSM_Clamp_WorldGroupSettings
   texture=/BRCosmetics/Characters/Player/Male/Medium/Bodies/M_MED_SpecterDunk/Textures/T_SpecterDunk_Alpha_02
  }
  n9 = TextureCoordinate()
  n10 = AppendVector(
   A=ScalarParameter() {defaultValue=0.335, group=Ranked, parameterName=Chest_UV_OffsetX}
   B=ScalarParameter() {defaultValue=0.34, group=Ranked, parameterName=Chest_UV_OffsetY}
  )
  n11 = ScalarParameter() {defaultValue=0.24, group=Ranked, parameterName=Chest_UV_ScaleX}
  n12 = ScalarParameter() {defaultValue=0.25, parameterName=Chest_UV_ScaleY}
  n13 = TextureSample(UVs=MaterialFunctionCall(UVs=Add(A=@n9, B=@n10), Texture Scale=AppendVector(A=@n11, B=@n12)) {materialFunction=&ScaleUVsByCenter}.UVs) {
   samplerSource=SSM_Clamp_WorldGroupSettings
   texture=/BRCosmetics/Characters/Player/Male/Medium/Bodies/M_MED_SpecterDunk/Textures/T_SpecterDunk_Alpha_01
  }
  n14 = NamedRerouteUsage()
  n15 = LinearInterpolate(A=NamedRerouteUsage(), B=Multiply(A=Add(A=@n8.RGB, B=@n13.RGB), B=@n14), Alpha=Add(A=@n8.A, B=@n13.A))
  n16 = Power(Base=@n14, Exp=Constant() {r=2.5})
  n17 = SphereMask(
   A=VertexInterpolator(VS=LocalPosition() {localOrigin=InstancePreSkinning}.XYZ).PS
   B=Constant4Vector() {constant=<rgba 53.2776 7.5449 90.954 1>}.RGB
   Radius=Constant() {r=50}
   Hardness=Constant() {r=0.7}
  )
  n18 = LinearInterpolate(B=LinearInterpolate(B=Saturate(None=@n16), Alpha=OneMinus(None=@n17)) {constA=1}, Alpha=NamedRerouteUsage()) {constA=1}
  NamedRerouteDeclaration(None=LinearInterpolate(A=@n15, B=Multiply(A=@n15, B=@n18)) {constAlpha=0.6}) {name=Out_BaseColor, nodeColor=<rgba 0 0.3438 1 1>}
  NamedRerouteDeclaration(None=@n3.R) {name=MetalsMask, nodeColor=<rgba 0 0.3438 1 1>}
  n19 = TextureSample() {
   samplerType=SAMPLERTYPE_Masks
   texture=/BRCosmetics/Characters/Player/Male/Medium/Bodies/M_MED_SpecterDunk/Textures/T_SpecterDunk_Body_TAMask2
  }
  NamedRerouteDeclaration(None=Power(Base=@n19.B)) {name=SkinDark, nodeColor=<rgba 0 0.0391 1 1>}
  NamedRerouteDeclaration(None=@n3.G) {name=SkullMask, nodeColor=<rgba 0.6641 1 0 1>}
  NamedRerouteUsage()
  NamedRerouteUsage()
 ]
}
