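# Physics asset for the two hoodie drawstrings on the male medium SpecterDunk cosmetic body.
# Each string is a chain of three dynamic-bone capsules hanging from a small base capsule constrained to spine_05, alongside larger chest and clavicle capsules.
# Chain joints allow wide swing on one axis, narrow swing on the other and almost no twist, while base and clavicle constraints to the spine are angularly free.
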PhysicsAsset /BRCosmetics/Characters/Player/Male/Medium/Bodies/M_MED_SpecterDunk/Meshes/M_MED_SpecterDunk_HoodieString_Physics
bodies {
 base_hoodie_string_l_1 [{shapeType=Capsule, length=0.505, radius=0.505, shapeName=base_hoodie_string_l_1_capsule, rotation=<rot 0 -90 0>}]
 base_hoodie_string_r_1 [{shapeType=Capsule, length=0.505, radius=0.505, shapeName=base_hoodie_string_r_1_capsule, rotation=<rot 0 -90 180>}]
 clavicle_l [{
   shapeType=Capsule
   length=10
   radius=10
   shapeName=clavicle_l_capsule
   center=<7.448 1.6843 -5.3427>
   rotation=<rot -87.59 179.9961 -179.9961>
  }]
 clavicle_r [{shapeType=Capsule, length=10, radius=10, shapeName=clavicle_r_capsule, center=<-7.448 -1.6843 5.3427>, rotation=<rot -87.59 179.9961 0.0039>}]
 dyn_hoodie_string_l_1 [{
   shapeType=Capsule
   length=6.1111
   radius=1.296
   shapeName=dyn_hoodie_string_l_1_capsule
   center=<3.7862 0.1667 0.0937>
   rotation=<rot 0 86.9634 -90.3181>
  }]
 dyn_hoodie_string_l_2 [{
   shapeType=Capsule
   length=5.4473
   radius=1.2896
   shapeName=dyn_hoodie_string_l_2_capsule
   center=<3.2239 0.0865 0.2312>
   rotation=<rot 0 89.3075 -91.0292>
  }]
 dyn_hoodie_string_l_3 [{
   shapeType=Capsule
   length=7.0265
   radius=1.7112
   shapeName=dyn_hoodie_string_l_3_capsule
   center=<2.51 0.0985 -0.3765>
   rotation=<rot 4.0383 -87.2163 -72.5263>
  }]
 dyn_hoodie_string_r_1 [{
   shapeType=Capsule
   length=6.1111
   radius=1.296
   shapeName=dyn_hoodie_string_r_1_capsule
   center=<-3.7862 -0.1667 -0.0937>
   rotation=<rot 0 86.9634 89.6819>
  }]
 dyn_hoodie_string_r_2 [{
   shapeType=Capsule
   length=5.4473
   radius=1.2896
   shapeName=dyn_hoodie_string_r_2_capsule
   center=<-3.2239 -0.0865 -0.2312>
   rotation=<rot 0 89.3075 88.9708>
  }]
 dyn_hoodie_string_r_3 [{
   shapeType=Capsule
   length=7.0265
   radius=1.7112
   shapeName=dyn_hoodie_string_r_3_capsule
   center=<-2.51 -0.0985 0.3765>
   rotation=<rot 4.0383 -87.2163 107.4737>
  }]
 spine_05 [{
   shapeType=Capsule
   length=16.6654
   radius=12.8333
   shapeName=spine_05_capsule
   center=<-4.8102 -0.1108 -0.1366>
   rotation=<rot -0.0469 -90 0.2061>
  }]
}
constraints [
 {
  bone1Name=base_hoodie_string_l_1
  bone2Name=spine_05
  swing1LimitDegrees=45
  swing1Motion=Free
  swing2LimitDegrees=45
  swing2Motion=Free
  twistLimitDegrees=45
  twistMotion=Free
 }
 {bone1Name=dyn_hoodie_string_l_1, bone2Name=base_hoodie_string_l_1, swing1LimitDegrees=10, swing2LimitDegrees=75, twistLimitDegrees=1}
 {bone1Name=dyn_hoodie_string_l_2, bone2Name=dyn_hoodie_string_l_1, swing1LimitDegrees=15, swing2LimitDegrees=60, twistLimitDegrees=1}
 {bone1Name=dyn_hoodie_string_l_3, bone2Name=dyn_hoodie_string_l_2, swing1LimitDegrees=15, swing2LimitDegrees=60, twistLimitDegrees=1}
 {
  bone1Name=base_hoodie_string_r_1
  bone2Name=spine_05
  swing1LimitDegrees=45
  swing1Motion=Free
  swing2LimitDegrees=45
  swing2Motion=Free
  twistLimitDegrees=45
  twistMotion=Free
 }
 {bone1Name=dyn_hoodie_string_r_1, bone2Name=base_hoodie_string_r_1, swing1LimitDegrees=10, swing2LimitDegrees=75, twistLimitDegrees=1}
 {bone1Name=dyn_hoodie_string_r_2, bone2Name=dyn_hoodie_string_r_1, swing1LimitDegrees=15, swing2LimitDegrees=60, twistLimitDegrees=1}
 {bone1Name=dyn_hoodie_string_r_3, bone2Name=dyn_hoodie_string_r_2, swing1LimitDegrees=15, swing2LimitDegrees=60, twistLimitDegrees=1}
 {
  bone1Name=clavicle_l
  bone2Name=spine_05
  swing1LimitDegrees=45
  swing1Motion=Free
  swing2LimitDegrees=45
  swing2Motion=Free
  twistLimitDegrees=45
  twistMotion=Free
 }
 {
  bone1Name=clavicle_r
  bone2Name=spine_05
  swing1LimitDegrees=45
  swing1Motion=Free
  swing2LimitDegrees=45
  swing2Motion=Free
  twistLimitDegrees=45
  twistMotion=Free
 }
]
