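# Physics asset for the male medium SpecterDunk body cosmetic, set up for three bombs dangling from its belt at front, middle and back.
# Each chain runs from a base attach body locked in rotation to the pelvis, through small belt capsules to a bomb capsule.
# Joints limit swing and allow almost no twist.
# Pelvis, leg and foot capsules supply body shapes, and a huge sphere below the root forms a floor-like surface.
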PhysicsAsset /BRCosmetics/Characters/Player/Male/Medium/Bodies/M_MED_SpecterDunk/Meshes/M_MED_SpecterDunk_Physics1
bodies {
 base_belt_back_bombattch_01 [{shapeType=Capsule, length=0.505, radius=0.505, shapeName=base_belt_back_bombattch_01_capsule, rotation=<rot 0 -90 0>}]
 base_belt_front_bombattch_1 [{shapeType=Capsule, length=0.505, radius=0.505, shapeName=base_belt_front_bombattch_1_capsule, rotation=<rot 0 -90 0>}]
 base_belt_mid_bombattch_01 [{shapeType=Capsule, length=0.505, radius=0.505, shapeName=base_belt_mid_bombattch_01_capsule, rotation=<rot 0 -90 0>}]
 calf_l [{
   shapeType=Capsule
   length=46.3764
   radius=9.8437
   shapeName=calf_l_capsule
   center=<-22.1552 -0.2886 0.2934>
   rotation=<rot -1.3239 -91.5196 -85.0517>
  }]
 calf_r [{
   shapeType=Capsule
   length=46.3764
   radius=9.8437
   shapeName=calf_r_capsule
   center=<22.1552 0.2886 -0.2934>
   rotation=<rot -1.3239 -91.5196 94.9483>
  }]
 dyn_back_bomb_1 [{shapeType=Capsule, length=1, radius=5, shapeName=dyn_mid_bomb_1_capsule, center=<5.5 -2 0>, rotation=<rot 0 90 90>}]
 dyn_belt_back_bomb_belt_1 [{shapeType=Capsule, length=4, radius=1.5, shapeName=dyn_belt_mid_bomb_belt_1_capsule, center=<4 0 0>, rotation=<rot 0 90 90>}]
 dyn_belt_back_bomb_belt_2 [{shapeType=Capsule, length=4, radius=1.5, shapeName=dyn_belt_mid_bomb_belt_1_capsule, center=<4 0 0>, rotation=<rot 0 90 90>}]
 dyn_belt_back_bomb_belt_3 [{shapeType=Capsule, length=4, radius=1.5, shapeName=dyn_belt_mid_bomb_belt_2_capsule, center=<4 0 0>, rotation=<rot 0 90 90>}]
 dyn_belt_back_bombattch_01 [{shapeType=Capsule, length=4, radius=1.5, shapeName=dyn_belt_mid_bombattch_01_capsule, center=<4 0 0>, rotation=<rot 0 90 90>}]
 dyn_belt_front_bomb_belt_1 [{shapeType=Capsule, length=4, radius=1.5, shapeName=dyn_belt_mid_bomb_belt_1_capsule, center=<4 0 0>, rotation=<rot 0 90 90>}]
 dyn_belt_front_bomb_belt_2 [{shapeType=Capsule, length=4, radius=1.5, shapeName=dyn_belt_mid_bomb_belt_2_capsule, center=<4 0 0>, rotation=<rot 0 90 90>}]
 dyn_belt_front_bombattch_1 [{shapeType=Capsule, length=4, radius=1.5, shapeName=dyn_belt_mid_bombattch_01_capsule, center=<4 0 0>, rotation=<rot 0 90 90>}]
 dyn_belt_mid_bomb_belt_1 [{shapeType=Capsule, length=4, radius=1.5, shapeName=dyn_belt_mid_bomb_belt_1_capsule, center=<4 0 0>, rotation=<rot 0 90 90>}]
 dyn_belt_mid_bomb_belt_2 [{shapeType=Capsule, length=4, radius=1.5, shapeName=dyn_belt_mid_bomb_belt_2_capsule, center=<4 0 0>, rotation=<rot 0 90 90>}]
 dyn_belt_mid_bombattch_01 [{shapeType=Capsule, length=4, radius=1.5, shapeName=dyn_belt_mid_bombattch_01_capsule, center=<4 0 0>, rotation=<rot 0 90 90>}]
 dyn_front_bomb_1 [{shapeType=Capsule, length=1, radius=5, shapeName=dyn_mid_bomb_1_capsule, center=<7 -1 0>, rotation=<rot 0 90 90>}]
 dyn_mid_bomb_1 [{shapeType=Capsule, length=3, radius=5, shapeName=dyn_mid_bomb_1_capsule, center=<5 -2 0.6>, rotation=<rot -1.1039 -94.8391 -89.5006>}]
 foot_l [{shapeType=Capsule, length=17.0574, radius=8.5505, shapeName=foot_l_capsule, center=<-3.7015 6.0795 0.4216>, rotation=<rot 0 3.2459 -89.2898>}]
 foot_r [{shapeType=Capsule, length=17.0574, radius=8.5505, shapeName=foot_r_capsule, center=<3.7015 -6.0795 -0.4216>, rotation=<rot 0 3.2459 90.7102>}]
 pelvis [{shapeType=Capsule, length=24.4764, radius=14, shapeName=pelvis_capsule, center=<6.8759 -0.3095 0>, rotation=<rot 0 -92.0455 -88.2175>}]
 root [{shapeType=Sphere, radius=281.7731, shapeName=root_sphere, center=<0 0 -277.4471>}]
 thigh_l [{
   shapeType=Capsule
   length=38.9618
   radius=10.1174
   shapeName=thigh_l_capsule
   center=<-18.9816 -0.5162 0.3044>
   rotation=<rot 0 -93.4274 -89.6306>
  }]
 thigh_r [{
   shapeType=Capsule
   length=38.9618
   radius=10.1174
   shapeName=thigh_r_capsule
   center=<18.9816 0.5162 -0.3044>
   rotation=<rot 0 -93.4274 90.3694>
  }]
}
constraints [
 {bone1Name=dyn_belt_back_bomb_belt_1, bone2Name=dyn_belt_back_bombattch_01, swing1LimitDegrees=45, swing2LimitDegrees=45, twistLimitDegrees=1}
 {bone1Name=dyn_belt_back_bomb_belt_2, bone2Name=dyn_belt_back_bomb_belt_1, swing1LimitDegrees=25, swing2LimitDegrees=25, twistLimitDegrees=1}
 {bone1Name=dyn_belt_back_bomb_belt_3, bone2Name=dyn_belt_back_bomb_belt_2, swing1LimitDegrees=25, swing2LimitDegrees=25, twistLimitDegrees=1}
 {bone1Name=dyn_back_bomb_1, bone2Name=dyn_belt_back_bomb_belt_3, swing1LimitDegrees=25, swing2LimitDegrees=25, twistLimitDegrees=1}
 {
  bone1Name=dyn_belt_front_bomb_belt_1
  bone2Name=dyn_belt_front_bombattch_1
  swing1LimitDegrees=45
  swing2LimitDegrees=45
  twistLimitDegrees=1
 }
 {
  bone1Name=dyn_belt_front_bomb_belt_2
  bone2Name=dyn_belt_front_bomb_belt_1
  swing1LimitDegrees=15
  swing2LimitDegrees=15
  twistLimitDegrees=1
 }
 {bone1Name=dyn_front_bomb_1, bone2Name=dyn_belt_front_bomb_belt_2, swing1LimitDegrees=15, swing2LimitDegrees=15, twistLimitDegrees=1}
 {bone1Name=dyn_belt_mid_bomb_belt_1, bone2Name=dyn_belt_mid_bombattch_01, swing1LimitDegrees=45, swing2LimitDegrees=45, twistLimitDegrees=1}
 {bone1Name=dyn_belt_mid_bomb_belt_2, bone2Name=dyn_belt_mid_bomb_belt_1, swing1LimitDegrees=25, swing2LimitDegrees=25, twistLimitDegrees=1}
 {
  bone1Name=dyn_belt_back_bombattch_01
  bone2Name=base_belt_back_bombattch_01
  swing1LimitDegrees=45
  swing2LimitDegrees=45
  twistLimitDegrees=1
 }
 {
  bone1Name=dyn_belt_front_bombattch_1
  bone2Name=base_belt_front_bombattch_1
  swing1LimitDegrees=45
  swing2LimitDegrees=45
  twistLimitDegrees=1
 }
 {bone1Name=dyn_belt_mid_bombattch_01, bone2Name=base_belt_mid_bombattch_01, swing1LimitDegrees=45, swing2LimitDegrees=45, twistLimitDegrees=1}
 {
  bone1Name=base_belt_back_bombattch_01
  bone2Name=pelvis
  swing1LimitDegrees=45
  swing1Motion=Locked
  swing2LimitDegrees=45
  swing2Motion=Locked
  twistLimitDegrees=45
  twistMotion=Locked
 }
 {
  bone1Name=base_belt_mid_bombattch_01
  bone2Name=pelvis
  swing1LimitDegrees=45
  swing1Motion=Locked
  swing2LimitDegrees=45
  swing2Motion=Locked
  twistLimitDegrees=45
  twistMotion=Locked
 }
 {
  bone1Name=base_belt_front_bombattch_1
  bone2Name=pelvis
  swing1LimitDegrees=45
  swing1Motion=Locked
  swing2LimitDegrees=45
  swing2Motion=Locked
  twistLimitDegrees=45
  twistMotion=Locked
 }
 {bone1Name=dyn_mid_bomb_1, bone2Name=dyn_belt_mid_bomb_belt_2, swing1LimitDegrees=25, swing2LimitDegrees=25, twistLimitDegrees=1}
]
