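# Animation Blueprint for the male medium SpecterDunk body cosmetic, driving secondary motion on its mesh.
# The attached parent's pose passes through a PreN Control Rig, constraints blending belt bomb-attach bones between dynamic belt bones, and an additive dyn_belt_main offset.
# Two chained RigidBody simulations follow: a hoodie-string asset based on spine_05, then a second physics asset based on pelvis.
# Each update computes and stores gravity override and rigid-body settings from a parameter asset.
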AnimBlueprint /BRCosmetics/Characters/Player/Male/Medium/Bodies/M_MED_SpecterDunk/Meshes/M_MED_SpecterDunk_AnimBP
target_skeleton=&/M_MED_SpecterDunk_Skeleton
graphs [
 AnimGraph {nodes [
   n1 = Control Rig(Source=Copy Pose From Mesh() {bUseAttachedParent=true}) {controlRigAssetReference {blueprintRigClass=&/M_MED_SpecterDunk_PreN_CtrlRig.M_MED_SpecterDunk_PreN_CtrlRig_C}}
   n2 = "Constraint - base_belt_front_bombattch_1 (PP)"!(ComponentPose=Local To Component(LocalPose=@n1), Alpha=1, ConstraintWeights_0=0.474, ConstraintWeights_1:float=0.526) -> (Pose) {
    boneToModify {boneName=base_belt_front_bombattch_1}
    constraintSetup [{targetBone {boneName=dyn_Belt_belt_frnt_l}}, {targetBone {boneName=dyn_Belt_belt_side_l}}]
    constraintWeights [1,1]
   }
   n3 = "Constraint - base_belt_mid_bombattch_01 (PP)"!(ComponentPose=@n2, Alpha=1, ConstraintWeights_0=0.022, ConstraintWeights_1:float=0.978) -> (Pose) {
    boneToModify {boneName=base_belt_mid_bombattch_01}
    constraintSetup [{targetBone {boneName=dyn_Belt_belt_frnt_l}}, {targetBone {boneName=dyn_Belt_belt_side_l}}]
    constraintWeights [1,1]
   }
   n4 = "Constraint - base_belt_back_bombattch_01 (PP)"!(ComponentPose=@n3, Alpha=1, ConstraintWeights_0=0.304, ConstraintWeights_1:float=0.696) -> (Pose) {
    boneToModify {boneName=base_belt_back_bombattch_01}
    constraintSetup [{targetBone {boneName=dyn_Belt_belt_back_l}}, {targetBone {boneName=dyn_Belt_belt_side_l}}]
    constraintWeights [1,1]
   }
   n5 = "Transform (Modify) Bone - Bone: dyn_belt_main"(ComponentPose=@n4, Alpha=1, Translation="-1.5,0,0", Rotation="-10,0,0", Scale="1,1,1") {
    rotationMode=BMM_Additive
    rotationSpace=BCS_BoneSpace
    translationMode=BMM_Additive
    translationSpace=BCS_BoneSpace
    boneToModify {boneName=dyn_belt_main}
    rotation {pitch=-10}
    translation {x=-1.5}
   }
   n6 = RigidBody!(
    ComponentPose=@n5
    Alpha
    OverrideWorldGravity:Vector="0,0,0"
    ExternalForce
    SimSpaceSettings:Sim Space Settings Structure=(WorldAlpha=0,VelocityScaleZ=1,DampingAlpha=1,MaxLinearVelocity=10000,MaxAngularVelocity=10000,MaxLinearAcceleration=10000,MaxAngularAcceleration=10000,ExternalLinearDragV=(X=0,Y=0,Z=0),ExternalLinearVelocity=(X=0,Y=0,Z=0),ExternalAngularVelocity=(X=0,Y=0,Z=0))
   ) -> (Pose) {
    bOverrideWorldGravity=true
    overridePhysicsAsset=&/M_MED_SpecterDunk_HoodieString_Physics
    simulationSpace=BaseBoneSpace
    baseBoneRef {boneName=spine_05}
   }
   n7 = RigidBody(
    ComponentPose=@n6
    OverrideWorldGravity="0,0,0"
    SimSpaceSettings=(WorldAlpha=0,VelocityScaleZ=1,DampingAlpha=1,MaxLinearVelocity=10000,MaxAngularVelocity=10000,MaxLinearAcceleration=10000,MaxAngularAcceleration=10000,ExternalLinearDragV=(X=0,Y=0,Z=0),ExternalLinearVelocity=(X=0,Y=0,Z=0),ExternalAngularVelocity=(X=0,Y=0,Z=0))
   ) {
    bOverrideWorldGravity=true
    overridePhysicsAsset=&/M_MED_SpecterDunk_Physics1
    simulationSpace=BaseBoneSpace
    alphaScaleBiasClamp {interpSpeedDecreasing=5, interpSpeedIncreasing=5}
    baseBoneRef {boneName=pelvis}
   }
   Output Pose(Result=Component To Local(ComponentPose=@n7)) {layerGroup=None}
  ]}
 EventGraph {nodes [
   Event BlueprintUpdateAnimation()
   > Sequence!(execute) -> (then_0, then_1, then_2:Exec)
   > n1 = "Gravity Override State Function (Parameter Asset) (Nativized)"(
    CharacterPartAnimInstance=Self-Reference()
    GameTimeInSeconds=GetGameTimeInSeconds()
    Parameters=&/M_MED_SpecterDunk_Parameters
   )
   > Set Gravity Override(Gravity Override=@n1)
   > Set RBAN Settings Out(RBAN Settings Out=@n1.OutRigidBodyAnimNodeParameters)
  ]}
]
variables [
 Gravity Override {type=Vector}
 RBAN Settings Out {
  type=FortRigidBodyAnimNodeParameters
  default {simSpaceSettings {
    maxAngularAcceleration=40
    maxAngularVelocity=8
    maxLinearAcceleration=4000
    maxLinearVelocity=800
    worldAlpha=1
    externalLinearDragV=<3 3 3>
   }}
 }
]
