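# Physics asset of cape colliders for the SpecterDunk medium male player body, named as collision for the character's cape.
# Capsules cover the pelvis, spine, neck and clavicles, with a sphere on the head.
# More capsules sit on the dynamic front, mid and back bomb bones and their belt and attachment bones.
# Arm, leg and foot bodies carry no shapes, and no constraints link the bodies.
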PhysicsAsset /BRCosmetics/Characters/Player/Male/Medium/Bodies/M_MED_SpecterDunk/Meshes/M_MED_SpecterDunk_CapeColliders
bodies {
 calf_l []
 calf_r []
 clavicle_l [{shapeType=Capsule, length=9.0192, radius=9.8871, center=<9.361 0.1223 -2.5021>, rotation=<rot 69.3297 1.4476 176.9462>}]
 clavicle_r [{shapeType=Capsule, length=9.0192, radius=9.8871, center=<-9.361 -0.1223 2.5021>, rotation=<rot 69.3297 1.4476 -3.0538>}]
 dyn_back_bomb_1 [{
   shapeType=Capsule
   length=16.5005
   radius=7.2777
   shapeName=dyn_front_bomb_1_capsule
   center=<-0.1593 2.8032 1.7707>
   rotation=<rot -79.2369 -27.7041 -2.334>
  }]
 dyn_belt_back_bomb_belt_1 [{
   shapeType=Capsule
   length=4.0809
   radius=2.9191
   shapeName=dyn_belt_back_bomb_belt_1_capsule
   center=<3.7129 0 0.4657>
   rotation=<rot -87.1853 -0.0004 0.0003>
  }]
 dyn_belt_back_bomb_belt_2 [{
   shapeType=Capsule
   length=6.5069
   radius=2.9274
   shapeName=dyn_belt_back_bomb_belt_2_capsule
   center=<6.609 -0.1112 0.1299>
   rotation=<rot -88.874 -0.3712 0.3598>
  }]
 dyn_belt_back_bombattch_01 [{shapeType=Capsule, length=4.3743, radius=2.6257, shapeName=dyn_belt_back_bombattch_01_capsule, center=<0 0 -3.3894>}]
 dyn_belt_front_bomb_belt_1 [{
   shapeType=Capsule
   length=5
   radius=2
   shapeName=dyn_belt_front_bomb_belt_1_capsule
   center=<1.077 0.0001 -0.1505>
   rotation=<rot -82.0479 179.9929 -179.9901>
  }]
 dyn_belt_front_bomb_belt_2 [{
   shapeType=Capsule
   length=5
   radius=2
   shapeName=dyn_belt_front_bomb_belt_2_capsule
   center=<2.7479 0 -0.4609>
   rotation=<rot -80.4792 179.9998 -179.9997>
  }]
 dyn_belt_front_bombattch_1 [{shapeType=Capsule, length=5, radius=2, shapeName=dyn_belt_front_bombattch_1_capsule, center=<0.4888 0 -1.4487>}]
 dyn_belt_mid_bomb_belt_1 [{
   shapeType=Capsule
   length=3.894
   radius=2.7992
   shapeName=dyn_belt_mid_bomb_belt_1_capsule
   center=<1.8544 -0.3977 -0.0562>
   rotation=<rot -88.264 -179.9998 179.9999>
  }]
 dyn_belt_mid_bomb_belt_2 [{
   shapeType=Capsule
   length=5.4973
   radius=2.5462
   shapeName=dyn_belt_mid_bomb_belt_2_capsule
   center=<1.8751 -0.1073 -0.4638>
   rotation=<rot 76.1124 -0.4941 2.6993>
  }]
 dyn_belt_mid_bombattch_01 [{shapeType=Capsule, length=3.7777, radius=2.5687, shapeName=dyn_belt_mid_bombattch_01_capsule, center=<0 0 -2.4055>}]
 dyn_front_bomb_1 [{
   shapeType=Capsule
   length=20.5825
   radius=6.1398
   shapeName=dyn_front_bomb_1_capsule
   center=<1.1899 2.0308 1.0345>
   rotation=<rot -89.1287 0.0098 -30.0098>
  }]
 dyn_mid_bomb_1 [{
   shapeType=Capsule
   length=17.5306
   radius=7.5242
   shapeName=dyn_front_bomb_1_capsule
   center=<-0.5807 2.3057 1.9312>
   rotation=<rot -69.2414 -28.8473 -1.2296>
  }]
 foot_l []
 foot_r []
 head [{shapeType=Sphere, radius=17.3125, center=<11.028 -0.174 0>}]
 lowerarm_l []
 lowerarm_r []
 neck_02 [{shapeType=Capsule, length=13.4375, radius=7.3175, center=<-0.2642 -0.3408 0>, rotation=<rot 90 -171.1193 -175.3978>}]
 pelvis [{shapeType=Capsule, length=19.9671, radius=18.5862, center=<-9.5685 -1.2847 0>}]
 spine_01 [{shapeType=Capsule, length=7.9438, radius=13.5586, center=<-1.7965 0.1884 0>}]
 spine_02 [
  {shapeType=Capsule, length=7.3264, radius=13.6322, center=<-1.711 -0.1851 0.0001>}
  {shapeType=Capsule, length=29.3956, radius=11.5807, center=<2.5024 0.4992 3.9825>, rotation=<rot 90 0.0049 0.0049>}
  {shapeType=Capsule, length=29.875, radius=11.1013, center=<2.4704 0.4991 -3.9825>, rotation=<rot 90 -0.0396 -0.0396>}
 ]
 spine_03 [{shapeType=Capsule, length=5.9816, radius=15.9321, center=<0.1445 0.2396 0.0001>}]
 spine_04 [{shapeType=Capsule, length=9.4404, radius=14.1325, center=<2.766 0.7765 0.0002>}]
 spine_05 [
  {shapeType=Capsule, length=8.0226, radius=15.695, center=<-2.384 3.1113 0>}
  {shapeType=Capsule, length=13.9518, radius=13.9425, center=<-0.3962 1.7456 0>}
 ]
 thigh_l []
 thigh_r []
 upperarm_l []
 upperarm_r []
}
constraints []
