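# Material function for the SpecterDunk character body materials that remaps UV0 into one tile of a 2x2 texture atlas.
# A Custom HLSL node halves the texture coordinates and adds an offset computed from the AtlasUV_2x2_Index scalar parameter (0-3).
# The code flips the row assuming a bottom-left UV origin, so indices 0-1 land in the upper-V half and 2-3 in the lower-V half.
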MaterialFunction /BRCosmetics/Characters/Player/Male/Medium/Bodies/M_MED_SpecterDunk/Materials/MF_AtlasUV_2x2
graph {nodes [
  MaterialExpressionTextureCoordinate_5 = TextureCoordinate()
  MaterialExpressionScalarParameter_0 = ScalarParameter() {group=AtlasUV_2x2, parameterName=AtlasUV_2x2_Index}
  n1 = Custom!(UV=@MaterialExpressionTextureCoordinate_5, Index=@MaterialExpressionScalarParameter_0) {
   code="// ----------------------------------------\r\n// 2x2 texture atlas UV selector\r\n// Index range: 0 - 3\r\n//\r\n// Index layout (left to right, top to bottom):\r\n//  0 | 1\r\n// ---+---\r\n//  2 | 3\r\n//\r\n// UV is assumed to be in 0-1 range\r\n// ----------------------------------------\r\n\r\nfloat2 tileSize = float2(0.5, 0.5);\r\n\r\n// Compute column and row from index\r\nfloat col = fmod(Index, 2.0);      // 0 or 1\r\nfloat row = floor(Index / 2.0);    // 0 or 1\r\n\r\n// Flip row because Unreal UV origin is bottom-left\r\nrow = 1.0 - row;\r\n\r\n// Calculate UV offset\r\nfloat2 offset = float2(col, row) * tileSize;\r\n\r\n// Scale UV into tile space and apply offset\r\nreturn UV * tileSize + offset;"
   outputType=CMOT_Float2
   showCode=true
   inputs [
    {
     inputName=UV
     input {expression=&self:MaterialExpressionTextureCoordinate_5}
    }
    {
     inputName=Index
     input {expression=&self:MaterialExpressionScalarParameter_0}
    }
   ]
  }
  FunctionOutput(None=@n1) {outputName=AtlasUV_2x2}
 ]}
