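# Material function in the SpecterDunk character's materials that outputs an animated grayscale noise value projected onto the surface.
# MF_TriplanarUV, borrowed from the BirchPalace Gas skin, supplies front and side UVs driven by parameters for scale, center, flip and clamp.
# Both UV sets pan slowly over a shared blurred noise texture.
# A boosted, saturated X-axis blend mask lerps between the two samples, so no top projection is used.
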MaterialFunction /BRCosmetics/Characters/Player/Male/Medium/Bodies/M_MED_SpecterDunk/Materials/MF_ProjectNoise
graph {nodes [
  n1 = MaterialFunctionCall!(
   TriplanarUV_UseClamp?=ScalarParameter() {group=TriplanarUV, parameterName=TriplanarUV_3D_UseClamp}
   TriplanarUV_FlipUV=ScalarParameter() {group=TriplanarUV, parameterName=TriplanarUV_3D_FlipUV}
   TriplanarUV_CenterPosition=VectorParameter() {group=TriplanarUV, parameterName=TriplanarUV_3D_CenterPosition, defaultValue=<rgba 0 0 0 1>}.RGB
   TriplanarUV_UVScale=ScalarParameter() {defaultValue=30, group=TriplanarUV, parameterName=TriplanarUV_3D_UVSize}
  ) {materialFunction=/BRCosmetics/Characters/Player/Male/Medium/Bodies/M_MED_BirchPalace/Skins/Gas/Materials/MF_TriplanarUV}
  n2 = TextureSample(UVs=Panner(Coordinate=NamedRerouteDeclaration(None=@n1.UV_FrontProjection) {name=TriplannarUV_front, nodeColor=<rgba 1 0.3516 0 1>}) {speedX=0.02, speedY=0.02}) {
   samplerSource=SSM_Wrap_WorldGroupSettings
   samplerType=SAMPLERTYPE_Grayscale
   texture=/Game/Global/Textures/Default/Noise/LowResBlurredNoiseNormalized
  }
  n3 = TextureSample(UVs=Panner(Coordinate=NamedRerouteDeclaration(None=@n1.UV_SideProjection) {name=TriplannarUV_Side, nodeColor=<rgba 0.1953 1 0 1>}) {speedX=0.02, speedY=0.02}) {
   samplerSource=SSM_Wrap_WorldGroupSettings
   samplerType=SAMPLERTYPE_Grayscale
   texture=&LowResBlurredNoiseNormalized
  }
  n4 = Multiply(A=MaterialFunctionCall() {materialFunction=/BRCosmetics/Characters/Player/Male/Medium/Bodies/M_MED_BirchPalace/Skins/Gas/Materials/MF_TriplanarBlendMask}.AxisMask_X) {constB=2}
  FunctionOutput(None=LinearInterpolate(A=@n2.R, B=@n3.R, Alpha=Saturate(None=@n4)))
 ]}
